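# Boss health HUD widget for PistolCanary: three stacked tentacle-material progress bars anchored top-center, hidden by default.
# UpdatePistolCanaryBossHealthBar, SetPistolCanaryBossHealthBarVisibility and SetBossHealthPercentage pick a bar by index through Select nodes to set percent or visibility.
# Custom events play per-bar fill-up and shake-and-flash damage animations; only bar 1 has an event for its long damage variant.
# Bar 1's fill event also sets it to full when finished.
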WidgetBlueprint /PistolCanary/Layout/BP/ComicPanels/PistolCanary_BossHealthBar
animations [
 HP_Fill {
  binding_count=1
  length_seconds=1
  object=&self:HP_Fill
  sequence {bindings [BossHealth1 {
     id=#1
     possessed_class=ProgressBar
     tracks [{
       type=MovieSceneFloatTrack
       display_name=Percent
       sections [{type=MovieSceneFloatSection, end_frame=20, start_frame=0, channels [None {type=MovieSceneScriptingFloatChannel, keys [[0,0],[20,1]]}]}]
      }]
    }]}
 }
 B1_TakeDamage {
  binding_count=1
  length_seconds=0.6
  object=&self:B1_TakeDamage
  sequence {
   bindings [BossHealth1 {
     id=#2
     possessed_class=ProgressBar
     tracks [
      {
       type=MovieScene2DTransformTrack
       sections [{
         type=MovieScene2DTransformSection
         end_frame=36
         start_frame=0
         channels [
          Translation.X {
           type=MovieSceneScriptingFloatChannel
           keys [[0,0],[1,-44.8],[5,-50],[9,0],[12,-50],[15,0],[18,-50],[21,0],[24,-44.8],[28,-50],[32,0],[35,-50],[36,0]]
          }
          Translation.Y {
           type=MovieSceneScriptingFloatChannel
           keys [[0,0],[1,8.96],[5,10],[9,0],[12,10],[15,0],[18,10],[21,0],[24,8.96],[28,10],[32,0],[35,10],[36,0]]
          }
          Angle {
           type=MovieSceneScriptingFloatChannel
           thinned_from=14
           keys [[0,0],[1,0],[5,-1],[9,0],[12,1],[18,-1],[21,0],[24,0],[28,-1],[32,0],[35,1],[35.999,0],[36,0]]
          }
          Scale.X {type=MovieSceneScriptingFloatChannel, const [1,1]}
          Scale.Y {type=MovieSceneScriptingFloatChannel, const [-1,1]}
          Shear.X {type=MovieSceneScriptingFloatChannel, const [0,1]}
          Shear.Y {type=MovieSceneScriptingFloatChannel, const [0,1]}
         ]
        }]
      }
      {
       type=MovieSceneColorTrack
       display_name=Fill Color and Opacity
       sections [{
         type=MovieSceneColorSection
         end_frame=36
         start_frame=0
         channels [
          Color.R {
           type=MovieSceneScriptingFloatChannel
           thinned_from=14
           keys [[0,1],[1,1],[5,0.25],[9,1],[12,0.25],[15,1],[18,0.25],[21,1],[24,1],[28,0.25],[32,1],[35,0.25],[36,1]]
          }
          Color.G {
           type=MovieSceneScriptingFloatChannel
           thinned_from=14
           keys [[0,1],[1,1],[5,0.25],[9,1],[12,0.25],[15,1],[18,0.25],[21,1],[24,1],[28,0.25],[32,1],[35,0.25],[36,1]]
          }
          Color.B {
           type=MovieSceneScriptingFloatChannel
           thinned_from=14
           keys [[0,1],[1,1],[5,0.25],[9,1],[12,0.25],[15,1],[18,0.25],[21,1],[24,1],[28,0.25],[32,1],[35,0.25],[36,1]]
          }
          Color.A {type=MovieSceneScriptingFloatChannel, const [1,13]}
         ]
        }]
      }
     ]
    }]
   display_rate {numerator=60}
  }
 }
 B1_TakeDamage_Long {
  binding_count=1
  length_seconds=1.2
  object=&self:B1_TakeDamage_Long
  sequence {
   bindings [BossHealth1 {
     id=#2
     possessed_class=ProgressBar
     tracks [
      {
       type=MovieScene2DTransformTrack
       sections [{
         type=MovieScene2DTransformSection
         end_frame=72
         start_frame=0
         channels [
          Translation.X {
           type=MovieSceneScriptingFloatChannel
           keys [[0,0],[1,-44.8],[5,-50],[9,0],[12,-50],[15,0],[18,-50],[21,0],[24,-44.8],[28,-50],[32,0],[35,-50],[38,0],[42,-44.8],[46,-50],[50,0],[53,-50],[56,0],[59,-50],[62,0],[65,-44.8],[69,-50],[72,0]]
          }
          Translation.Y {
           type=MovieSceneScriptingFloatChannel
           keys [[0,0],[1,8.96],[5,10],[9,0],[12,10],[15,0],[18,10],[21,0],[24,8.96],[28,10],[32,0],[35,10],[38,0],[42,8.96],[46,10],[50,0],[53,10],[56,0],[59,10],[62,0],[65,8.96],[69,10],[72,0]]
          }
          Angle {
           type=MovieSceneScriptingFloatChannel
           thinned_from=25
           keys [[0,0],[1,0],[5,-1],[9,0],[12,1],[18,-1],[21,0],[24,0],[28,-1],[32,0],[35,1],[37.999,0],[42,0],[46,-1],[50,0],[53,1],[59,-1],[62,0],[65,0],[69,-1],[71.999,0],[72,0]]
          }
          Scale.X {type=MovieSceneScriptingFloatChannel, const [1,1]}
          Scale.Y {type=MovieSceneScriptingFloatChannel, const [-1,1]}
          Shear.X {type=MovieSceneScriptingFloatChannel, const [0,1]}
          Shear.Y {type=MovieSceneScriptingFloatChannel, const [0,1]}
         ]
        }]
      }
      {
       type=MovieSceneColorTrack
       display_name=Fill Color and Opacity
       sections [{
         type=MovieSceneColorSection
         end_frame=72
         start_frame=0
         channels [
          Color.R {
           type=MovieSceneScriptingFloatChannel
           thinned_from=26
           keys [[0,1],[1,1],[5,0.25],[9,1],[12,0.25],[15,1],[18,0.25],[21,1],[24,1],[28,0.25],[32,1],[35,0.25],[38.001,1],[42,1],[46,0.25],[50,1],[53,0.25],[56,1],[59,0.25],[62,1],[65,1],[69,0.25],[72.001,1]]
          }
          Color.G {
           type=MovieSceneScriptingFloatChannel
           thinned_from=26
           keys [[0,1],[1,1],[5,0.25],[9,1],[12,0.25],[15,1],[18,0.25],[21,1],[24,1],[28,0.25],[32,1],[35,0.25],[38.001,1],[42,1],[46,0.25],[50,1],[53,0.25],[56,1],[59,0.25],[62,1],[65,1],[69,0.25],[72.001,1]]
          }
          Color.B {
           type=MovieSceneScriptingFloatChannel
           thinned_from=26
           keys [[0,1],[1,1],[5,0.25],[9,1],[12,0.25],[15,1],[18,0.25],[21,1],[24,1],[28,0.25],[32,1],[35,0.25],[38.001,1],[42,1],[46,0.25],[50,1],[53,0.25],[56,1],[59,0.25],[62,1],[65,1],[69,0.25],[72.001,1]]
          }
          Color.A {type=MovieSceneScriptingFloatChannel, const [1,23]}
         ]
        }]
      }
     ]
    }]
   display_rate {numerator=60}
  }
 }
 HP_Fill_Bar2 {
  binding_count=1
  length_seconds=1
  object=&self:HP_Fill_Bar2
  sequence {bindings [BossHealth2 {
     id=#3
     possessed_class=ProgressBar
     tracks [{
       type=MovieSceneFloatTrack
       display_name=Percent
       sections [{type=MovieSceneFloatSection, end_frame=20, start_frame=0, channels [None {type=MovieSceneScriptingFloatChannel, keys [[0,0],[20,1]]}]}]
      }]
    }]}
 }
 HP_Fill_Bar3 {
  binding_count=1
  length_seconds=1
  object=&self:HP_Fill_Bar3
  sequence {bindings [BossHealth3 {
     id=#4
     possessed_class=ProgressBar
     tracks [{
       type=MovieSceneFloatTrack
       display_name=Percent
       sections [{type=MovieSceneFloatSection, end_frame=20, start_frame=0, channels [None {type=MovieSceneScriptingFloatChannel, keys [[0,0],[20,1]]}]}]
      }]
    }]}
 }
 B2_TakeDamage {
  binding_count=1
  length_seconds=0.6
  object=&self:B2_TakeDamage
  sequence {
   bindings [BossHealth2 {
     id=#5
     possessed_class=ProgressBar
     tracks [
      {
       type=MovieScene2DTransformTrack
       sections [{
         type=MovieScene2DTransformSection
         end_frame=36
         start_frame=0
         channels [
          Translation.X {
           type=MovieSceneScriptingFloatChannel
           keys [[0,0],[1,-44.8],[5,-50],[9,0],[12,-50],[15,0],[18,-50],[21,0],[24,-44.8],[28,-50],[32,0],[35,-50],[36,0]]
          }
          Translation.Y {
           type=MovieSceneScriptingFloatChannel
           keys [[0,0],[1,8.96],[5,10],[9,0],[12,10],[15,0],[18,10],[21,0],[24,8.96],[28,10],[32,0],[35,10],[36,0]]
          }
          Angle {
           type=MovieSceneScriptingFloatChannel
           thinned_from=14
           keys [[0,0],[1,0],[5,-1],[9,0],[12,1],[18,-1],[21,0],[24,0],[28,-1],[32,0],[35,1],[35.999,0],[36,0]]
          }
          Scale.X {type=MovieSceneScriptingFloatChannel, const [1,1]}
          Scale.Y {type=MovieSceneScriptingFloatChannel, default=-1, const [0.75,1]}
          Shear.X {type=MovieSceneScriptingFloatChannel, const [0,1]}
          Shear.Y {type=MovieSceneScriptingFloatChannel, const [0,1]}
         ]
        }]
      }
      {
       type=MovieSceneColorTrack
       display_name=Fill Color and Opacity
       sections [{
         type=MovieSceneColorSection
         end_frame=36
         start_frame=0
         channels [
          Color.R {
           type=MovieSceneScriptingFloatChannel
           thinned_from=14
           keys [[0,1],[1,1],[5,0.25],[9,1],[12,0.25],[15,1],[18,0.25],[21,1],[24,1],[28,0.25],[32,1],[35,0.25],[36,1]]
          }
          Color.G {
           type=MovieSceneScriptingFloatChannel
           thinned_from=14
           keys [[0,1],[1,1],[5,0.25],[9,1],[12,0.25],[15,1],[18,0.25],[21,1],[24,1],[28,0.25],[32,1],[35,0.25],[36,1]]
          }
          Color.B {
           type=MovieSceneScriptingFloatChannel
           thinned_from=14
           keys [[0,1],[1,1],[5,0.25],[9,1],[12,0.25],[15,1],[18,0.25],[21,1],[24,1],[28,0.25],[32,1],[35,0.25],[36,1]]
          }
          Color.A {type=MovieSceneScriptingFloatChannel, const [1,13]}
         ]
        }]
      }
     ]
    }]
   display_rate {numerator=60}
  }
 }
 B2_TakeDamage_Long {
  binding_count=1
  length_seconds=1.2
  object=&self:B2_TakeDamage_Long
  sequence {
   bindings [BossHealth2 {
     id=#6
     possessed_class=ProgressBar
     tracks [
      {
       type=MovieScene2DTransformTrack
       sections [{
         type=MovieScene2DTransformSection
         end_frame=72
         start_frame=0
         channels [
          Translation.X {
           type=MovieSceneScriptingFloatChannel
           keys [[0,0],[1,-44.8],[5,-50],[9,0],[12,-50],[15,0],[18,-50],[21,0],[24,-44.8],[28,-50],[32,0],[35,-50],[38,0],[42,-44.8],[46,-50],[50,0],[53,-50],[56,0],[59,-50],[62,0],[65,-44.8],[69,-50],[72,0]]
          }
          Translation.Y {
           type=MovieSceneScriptingFloatChannel
           keys [[0,0],[1,8.96],[5,10],[9,0],[12,10],[15,0],[18,10],[21,0],[24,8.96],[28,10],[32,0],[35,10],[38,0],[42,8.96],[46,10],[50,0],[53,10],[56,0],[59,10],[62,0],[65,8.96],[69,10],[72,0]]
          }
          Angle {
           type=MovieSceneScriptingFloatChannel
           thinned_from=25
           keys [[0,0],[1,0],[5,-1],[9,0],[12,1],[18,-1],[21,0],[24,0],[28,-1],[32,0],[35,1],[37.999,0],[42,0],[46,-1],[50,0],[53,1],[59,-1],[62,0],[65,0],[69,-1],[71.999,0],[72,0]]
          }
          Scale.X {type=MovieSceneScriptingFloatChannel, const [1,1]}
          Scale.Y {type=MovieSceneScriptingFloatChannel, default=-1, const [0.75,1]}
          Shear.X {type=MovieSceneScriptingFloatChannel, const [0,1]}
          Shear.Y {type=MovieSceneScriptingFloatChannel, const [0,1]}
         ]
        }]
      }
      {
       type=MovieSceneColorTrack
       display_name=Fill Color and Opacity
       sections [{
         type=MovieSceneColorSection
         end_frame=72
         start_frame=0
         channels [
          Color.R {
           type=MovieSceneScriptingFloatChannel
           thinned_from=26
           keys [[0,1],[1,1],[5,0.25],[9,1],[12,0.25],[15,1],[18,0.25],[21,1],[24,1],[28,0.25],[32,1],[35,0.25],[38.001,1],[42,1],[46,0.25],[50,1],[53,0.25],[56,1],[59,0.25],[62,1],[65,1],[69,0.25],[72.001,1]]
          }
          Color.G {
           type=MovieSceneScriptingFloatChannel
           thinned_from=26
           keys [[0,1],[1,1],[5,0.25],[9,1],[12,0.25],[15,1],[18,0.25],[21,1],[24,1],[28,0.25],[32,1],[35,0.25],[38.001,1],[42,1],[46,0.25],[50,1],[53,0.25],[56,1],[59,0.25],[62,1],[65,1],[69,0.25],[72.001,1]]
          }
          Color.B {
           type=MovieSceneScriptingFloatChannel
           thinned_from=26
           keys [[0,1],[1,1],[5,0.25],[9,1],[12,0.25],[15,1],[18,0.25],[21,1],[24,1],[28,0.25],[32,1],[35,0.25],[38.001,1],[42,1],[46,0.25],[50,1],[53,0.25],[56,1],[59,0.25],[62,1],[65,1],[69,0.25],[72.001,1]]
          }
          Color.A {type=MovieSceneScriptingFloatChannel, const [1,23]}
         ]
        }]
      }
     ]
    }]
   display_rate {numerator=60}
  }
 }
 B3_TakeDamage {
  binding_count=1
  length_seconds=0.6
  object=&self:B3_TakeDamage
  sequence {
   bindings [BossHealth3 {
     id=#7
     possessed_class=ProgressBar
     tracks [
      {
       type=MovieScene2DTransformTrack
       sections [{
         type=MovieScene2DTransformSection
         end_frame=36
         start_frame=0
         channels [
          Translation.X {
           type=MovieSceneScriptingFloatChannel
           keys [[0,0],[1,-44.8],[5,-50],[9,0],[12,-50],[15,0],[18,-50],[21,0],[24,-44.8],[28,-50],[32,0],[35,-50],[36,0]]
          }
          Translation.Y {
           type=MovieSceneScriptingFloatChannel
           keys [[0,0],[1,8.96],[5,10],[9,0],[12,10],[15,0],[18,10],[21,0],[24,8.96],[28,10],[32,0],[35,10],[36,0]]
          }
          Angle {
           type=MovieSceneScriptingFloatChannel
           thinned_from=14
           keys [[0,0],[1,0],[5,-1],[9,0],[12,1],[18,-1],[21,0],[24,0],[28,-1],[32,0],[35,1],[35.999,0],[36,0]]
          }
          Scale.X {type=MovieSceneScriptingFloatChannel, const [1,1]}
          Scale.Y {type=MovieSceneScriptingFloatChannel, const [1,1]}
          Shear.X {type=MovieSceneScriptingFloatChannel, const [0,1]}
          Shear.Y {type=MovieSceneScriptingFloatChannel, const [0,1]}
         ]
        }]
      }
      {
       type=MovieSceneColorTrack
       display_name=Fill Color and Opacity
       sections [{
         type=MovieSceneColorSection
         end_frame=36
         start_frame=0
         channels [
          Color.R {
           type=MovieSceneScriptingFloatChannel
           thinned_from=14
           keys [[0,1],[1,1],[5,0.25],[9,1],[12,0.25],[15,1],[18,0.25],[21,1],[24,1],[28,0.25],[32,1],[35,0.25],[36,1]]
          }
          Color.G {
           type=MovieSceneScriptingFloatChannel
           thinned_from=14
           keys [[0,1],[1,1],[5,0.25],[9,1],[12,0.25],[15,1],[18,0.25],[21,1],[24,1],[28,0.25],[32,1],[35,0.25],[36,1]]
          }
          Color.B {
           type=MovieSceneScriptingFloatChannel
           thinned_from=14
           keys [[0,1],[1,1],[5,0.25],[9,1],[12,0.25],[15,1],[18,0.25],[21,1],[24,1],[28,0.25],[32,1],[35,0.25],[36,1]]
          }
          Color.A {type=MovieSceneScriptingFloatChannel, const [1,13]}
         ]
        }]
      }
     ]
    }]
   display_rate {numerator=60}
  }
 }
 B3_TakeDamage_Long {
  binding_count=1
  length_seconds=1.2
  object=&self:B3_TakeDamage_Long
  sequence {
   bindings [BossHealth3 {
     id=#8
     possessed_class=ProgressBar
     tracks [
      {
       type=MovieScene2DTransformTrack
       sections [{
         type=MovieScene2DTransformSection
         end_frame=72
         start_frame=0
         channels [
          Translation.X {
           type=MovieSceneScriptingFloatChannel
           keys [[0,0],[1,-44.8],[5,-50],[9,0],[12,-50],[15,0],[18,-50],[21,0],[24,-44.8],[28,-50],[32,0],[35,-50],[38,0],[42,-44.8],[46,-50],[50,0],[53,-50],[56,0],[59,-50],[62,0],[65,-44.8],[69,-50],[72,0]]
          }
          Translation.Y {
           type=MovieSceneScriptingFloatChannel
           keys [[0,0],[1,8.96],[5,10],[9,0],[12,10],[15,0],[18,10],[21,0],[24,8.96],[28,10],[32,0],[35,10],[38,0],[42,8.96],[46,10],[50,0],[53,10],[56,0],[59,10],[62,0],[65,8.96],[69,10],[72,0]]
          }
          Angle {
           type=MovieSceneScriptingFloatChannel
           thinned_from=25
           keys [[0,0],[1,0],[5,-1],[9,0],[12,1],[18,-1],[21,0],[24,0],[28,-1],[32,0],[35,1],[37.999,0],[42,0],[46,-1],[50,0],[53,1],[59,-1],[62,0],[65,0],[69,-1],[71.999,0],[72,0]]
          }
          Scale.X {type=MovieSceneScriptingFloatChannel, const [1,1]}
          Scale.Y {type=MovieSceneScriptingFloatChannel, const [1,1]}
          Shear.X {type=MovieSceneScriptingFloatChannel, const [0,1]}
          Shear.Y {type=MovieSceneScriptingFloatChannel, const [0,1]}
         ]
        }]
      }
      {
       type=MovieSceneColorTrack
       display_name=Fill Color and Opacity
       sections [{
         type=MovieSceneColorSection
         end_frame=72
         start_frame=0
         channels [
          Color.R {
           type=MovieSceneScriptingFloatChannel
           thinned_from=26
           keys [[0,1],[1,1],[5,0.25],[9,1],[12,0.25],[15,1],[18,0.25],[21,1],[24,1],[28,0.25],[32,1],[35,0.25],[38.001,1],[42,1],[46,0.25],[50,1],[53,0.25],[56,1],[59,0.25],[62,1],[65,1],[69,0.25],[72.001,1]]
          }
          Color.G {
           type=MovieSceneScriptingFloatChannel
           thinned_from=26
           keys [[0,1],[1,1],[5,0.25],[9,1],[12,0.25],[15,1],[18,0.25],[21,1],[24,1],[28,0.25],[32,1],[35,0.25],[38.001,1],[42,1],[46,0.25],[50,1],[53,0.25],[56,1],[59,0.25],[62,1],[65,1],[69,0.25],[72.001,1]]
          }
          Color.B {
           type=MovieSceneScriptingFloatChannel
           thinned_from=26
           keys [[0,1],[1,1],[5,0.25],[9,1],[12,0.25],[15,1],[18,0.25],[21,1],[24,1],[28,0.25],[32,1],[35,0.25],[38.001,1],[42,1],[46,0.25],[50,1],[53,0.25],[56,1],[59,0.25],[62,1],[65,1],[69,0.25],[72.001,1]]
          }
          Color.A {type=MovieSceneScriptingFloatChannel, const [1,23]}
         ]
        }]
      }
     ]
    }]
   display_rate {numerator=60}
  }
 }
]
graphs [
 EventGraph {nodes [
   Anim_HP_Fill_Bar1() {is=title K2Node_CustomEvent}
   > n1 = Play Animation with Finished event!(
    execute:Exec
    Widget:User Widget Object Reference=Self-Reference()
    InAnimation:Widget Animation Object Reference=Get HP_Fill() -> (HP_Fill:Widget Animation Object Reference)
    StartAtTime:float=0
    NumLoopsToPlay:int=1
    PlayMode:EUMGSequencePlayMode Enum
    PlaybackSpeed:float=1
   ) -> (then:Exec, Result:Widget Animation Handle Structure, Finished:Exec) {is=title K2Node_PlayAnimation2}
   SetPercent!(execute=@n1.Finished, self:Progress Bar Object Reference=Get BossHealth1() -> (BossHealth1:Progress Bar Object Reference), InPercent:float=1) -> (then)
   Anim_TakeDamage_B3() {is=title K2Node_CustomEvent}
   > Play Animation with Finished event(
    Widget=Self-Reference()
    InAnimation=Get B3_TakeDamage() -> (B3_TakeDamage:Widget Animation Object Reference)
    StartAtTime=0
    NumLoopsToPlay=1
    PlaybackSpeed=1
   )
   Anim_TakeDamage_Long_B1() {is=title K2Node_CustomEvent}
   > Play Animation with Finished event(
    Widget=Self-Reference()
    InAnimation=Get B1_TakeDamage_Long() -> (B1_TakeDamage_Long:Widget Animation Object Reference)
    StartAtTime=0
    NumLoopsToPlay=1
    PlaybackSpeed=1
   )
   Anim_HP_Fill_Bar2() {is=title K2Node_CustomEvent}
   > Play Animation with Finished event(
    Widget=Self-Reference()
    InAnimation=Get HP_Fill_Bar2() -> (HP_Fill_Bar2:Widget Animation Object Reference)
    StartAtTime=0
    NumLoopsToPlay=1
    PlaybackSpeed=1
   )
   Anim_HP_Fill_Bar3() {is=title K2Node_CustomEvent}
   > Play Animation with Finished event(
    Widget=Self-Reference()
    InAnimation=Get HP_Fill_Bar3() -> (HP_Fill_Bar3:Widget Animation Object Reference)
    StartAtTime=0
    NumLoopsToPlay=1
    PlaybackSpeed=1
   )
   Anim_TakeDamage_B1() {is=title K2Node_CustomEvent}
   > Play Animation with Finished event(
    Widget=Self-Reference()
    InAnimation=Get B1_TakeDamage() -> (B1_TakeDamage:Widget Animation Object Reference)
    StartAtTime=0
    NumLoopsToPlay=1
    PlaybackSpeed=1
   )
   Anim_TakeDamage_B2() {is=title K2Node_CustomEvent}
   > Play Animation with Finished event(
    Widget=Self-Reference()
    InAnimation=Get B2_TakeDamage() -> (B2_TakeDamage:Widget Animation Object Reference)
    StartAtTime=0
    NumLoopsToPlay=1
    PlaybackSpeed=1
   )
   n2 = Event UpdatePistolCanaryBossHealthBar!() -> (OutputDelegate, then, HealthBarIndex:int=0, Percent:double) {is=title K2Node_Event}
   n3 = Get BossHealth1()
   n4 = Get BossHealth2() -> (BossHealth2:Progress Bar Object Reference)
   n5 = Get BossHealth3() -> (BossHealth3:Progress Bar Object Reference)
   > SetPercent(
    self=Select!(
     Option 0:Progress Bar Object Reference=@n3
     Option 1:Progress Bar Object Reference=@n4
     Option 2:Progress Bar Object Reference=@n5
     Index:int=@n2.HealthBarIndex
    ) -> (ReturnValue:Progress Bar Object Reference)
    InPercent=@n2.Percent
   )
   n6 = Event SetPistolCanaryBossHealthBarVisibility!() -> (OutputDelegate, then, HealthBarIndex:int=0, Visibility:ESlateVisibility Enum=Visible) {is=title K2Node_Event}
   > SetVisibility(
    self=Select(Option 0:Widget Object Reference=@n3, Option 1:Widget Object Reference=@n4, Option 2:Widget Object Reference=@n5, Index=@n6.HealthBarIndex) -> (ReturnValue:Widget Object Reference)
    InVisibility=@n6.Visibility
   )
   Event PreConstruct()
   Event Construct()
   Event Tick()
  ]}
 SetBossHealthPercentage {nodes [
   n1 = SetBossHealthPercentage!() -> (then, Index:int, Percent:float=~) {is=title K2Node_FunctionEntry}
   > SetPercent(
    self=Select(
     Option 0:Progress Bar Object Reference=Get BossHealth1()
     Option 1:Progress Bar Object Reference=Get BossHealth2()
     Option 2:Progress Bar Object Reference=Get BossHealth3()
     Index=@n1
    ) -> (ReturnValue:Progress Bar Object Reference)
    InPercent=@n1.Percent
   )
  ]}
]
variables []
widgets {
 info {inheritedWidgetCount=0, parentClass=&UserWidget, rootWidgetClass=&CanvasPanel, widgetCount=4}
 widgets [
  CanvasPanel_60 {widgetClassPath=&CanvasPanel, widget=CanvasPanel_60}
  BossHealth1 {
   widgetClassPath=/Script/UMG.ProgressBar
   bIsVariable=true
   parent=CanvasPanel_60
   widget=BossHealth1
   properties {
    percent=1
    visibility=Hidden
    renderTransform {angle=0, scale=<1 -1>, shear=<0 0>, translation=<0 0>}
    widgetStyle {
     enableFillAnimation=false
     backgroundImage {
      bIsDynamicallyLoaded=false
      drawAs=Image
      imageType=FullColor
      mirroring=NoMirror
      resourceName=None
      resourceObject=/PistolCanary/UI/MI_UI_TentacleBarBG
      tiling=NoTile
      imageSize=<0 0>
      margin=<ltrb 0.4167 0.4167 0.4167 0.4167>
      outlineSettings {
       bUseBrushTransparency=false
       roundingType=HalfHeightRadius
       width=0
       color {colorUseRule=UseColor_Specified, specifiedColor=<rgba 0 0 0 0>}
       cornerRadii=<0 0 0 0>
      }
      tintColor {colorUseRule=UseColor_Specified, specifiedColor=<rgba 1 0 0 0.4>}
      uVRegion {bIsValid=false, max=<0 0>, min=<0 0>}
     }
     fillImage {
      bIsDynamicallyLoaded=false
      drawAs=Image
      imageType=FullColor
      mirroring=NoMirror
      resourceName=None
      resourceObject=/PistolCanary/UI/MI_UI_TentacleBar
      tiling=NoTile
      imageSize=<0 0>
      margin=<ltrb 0.4167 0.4167 0.4167 0.4167>
      outlineSettings {
       bUseBrushTransparency=false
       roundingType=HalfHeightRadius
       width=0
       color {colorUseRule=UseColor_Specified, specifiedColor=<rgba 0 0 0 0>}
       cornerRadii=<0 0 0 0>
      }
      tintColor {colorUseRule=UseColor_Specified, specifiedColor=<rgba 1 1 1 1>}
      uVRegion {bIsValid=false, max=<0 0>, min=<0 0>}
     }
     marqueeImage {
      bIsDynamicallyLoaded=false
      drawAs=Image
      imageType=FullColor
      mirroring=NoMirror
      resourceName=../../../Engine/Content/Slate/Common/ProgressBar_Marquee.png
      resourceObject=null
      tiling=Horizontal
      imageSize=<20 12>
      margin=<ltrb 0 0 0 0>
      outlineSettings {
       bUseBrushTransparency=false
       roundingType=HalfHeightRadius
       width=0
       color {colorUseRule=UseColor_Specified, specifiedColor=<rgba 0 0 0 0>}
       cornerRadii=<0 0 0 0>
      }
      tintColor {colorUseRule=UseColor_Specified, specifiedColor=<rgba 1 1 1 1>}
      uVRegion {bIsValid=false, max=<0 0>, min=<0 0>}
     }
    }
   }
   slot {
    class=CanvasPanelSlot
    properties {layoutData {alignment=<0.5 0.5>, anchors {maximum=<0.5 0>, minimum=<0.5 0>}, offsets=<ltrb 0 100 1000 20>}}
   }
  }
  BossHealth2 {
   widgetClassPath=&ProgressBar
   bIsVariable=true
   parent=CanvasPanel_60
   widget=BossHealth2
   properties {
    percent=1
    visibility=Hidden
    borderPadding=<0 0.5>
    renderTransform {angle=0, scale=<1 0.75>, shear=<0 0>, translation=<0 0>}
    widgetStyle {
     enableFillAnimation=false
     backgroundImage {
      bIsDynamicallyLoaded=false
      drawAs=Box
      imageType=FullColor
      mirroring=NoMirror
      resourceName=None
      resourceObject=&MI_UI_TentacleBarBG
      tiling=NoTile
      imageSize=<0 0>
      margin=<ltrb 0 0 0 1>
      outlineSettings {
       bUseBrushTransparency=false
       roundingType=HalfHeightRadius
       width=0
       color {colorUseRule=UseColor_Specified, specifiedColor=<rgba 0 0 0 0>}
       cornerRadii=<0 0 0 0>
      }
      tintColor {colorUseRule=UseColor_Specified, specifiedColor=<rgba 1 0 0 0.4>}
      uVRegion {bIsValid=false, max=<0 0>, min=<0 0>}
     }
     fillImage {
      bIsDynamicallyLoaded=false
      drawAs=Box
      imageType=FullColor
      mirroring=NoMirror
      resourceName=None
      resourceObject=&MI_UI_TentacleBar
      tiling=NoTile
      imageSize=<0 0>
      margin=<ltrb 0 0 0 1>
      outlineSettings {
       bUseBrushTransparency=false
       roundingType=HalfHeightRadius
       width=0
       color {colorUseRule=UseColor_Specified, specifiedColor=<rgba 0 0 0 0>}
       cornerRadii=<0 0 0 0>
      }
      tintColor {colorUseRule=UseColor_Specified, specifiedColor=<rgba 1 1 1 1>}
      uVRegion {bIsValid=false, max=<0 0>, min=<0 0>}
     }
     marqueeImage {
      bIsDynamicallyLoaded=false
      drawAs=Image
      imageType=FullColor
      mirroring=NoMirror
      resourceName=../../../Engine/Content/Slate/Common/ProgressBar_Marquee.png
      resourceObject=null
      tiling=Horizontal
      imageSize=<20 12>
      margin=<ltrb 0 0 0 0>
      outlineSettings {
       bUseBrushTransparency=false
       roundingType=HalfHeightRadius
       width=0
       color {colorUseRule=UseColor_Specified, specifiedColor=<rgba 0 0 0 0>}
       cornerRadii=<0 0 0 0>
      }
      tintColor {colorUseRule=UseColor_Specified, specifiedColor=<rgba 1 1 1 1>}
      uVRegion {bIsValid=false, max=<0 0>, min=<0 0>}
     }
    }
   }
   slot {
    class=CanvasPanelSlot
    properties {layoutData {alignment=<0.5 0.5>, anchors {maximum=<0.5 0>, minimum=<0.5 0>}, offsets=<ltrb 0 140 1000 20>}}
   }
  }
  BossHealth3 {
   widgetClassPath=&ProgressBar
   bIsVariable=true
   parent=CanvasPanel_60
   widget=BossHealth3
   properties {
    percent=1
    visibility=Hidden
    widgetStyle {
     enableFillAnimation=false
     backgroundImage {
      bIsDynamicallyLoaded=false
      drawAs=Image
      imageType=FullColor
      mirroring=NoMirror
      resourceName=None
      resourceObject=&MI_UI_TentacleBarBG
      tiling=NoTile
      imageSize=<0 0>
      margin=<ltrb 0.4167 0.4167 0.4167 0.4167>
      outlineSettings {
       bUseBrushTransparency=false
       roundingType=HalfHeightRadius
       width=0
       color {colorUseRule=UseColor_Specified, specifiedColor=<rgba 0 0 0 0>}
       cornerRadii=<0 0 0 0>
      }
      tintColor {colorUseRule=UseColor_Specified, specifiedColor=<rgba 1 0 0 0.4>}
      uVRegion {bIsValid=false, max=<0 0>, min=<0 0>}
     }
     fillImage {
      bIsDynamicallyLoaded=false
      drawAs=Image
      imageType=FullColor
      mirroring=NoMirror
      resourceName=None
      resourceObject=&MI_UI_TentacleBar
      tiling=NoTile
      imageSize=<0 0>
      margin=<ltrb 0.4167 0.4167 0.4167 0.4167>
      outlineSettings {
       bUseBrushTransparency=false
       roundingType=HalfHeightRadius
       width=0
       color {colorUseRule=UseColor_Specified, specifiedColor=<rgba 0 0 0 0>}
       cornerRadii=<0 0 0 0>
      }
      tintColor {colorUseRule=UseColor_Specified, specifiedColor=<rgba 1 1 1 1>}
      uVRegion {bIsValid=false, max=<0 0>, min=<0 0>}
     }
     marqueeImage {
      bIsDynamicallyLoaded=false
      drawAs=Image
      imageType=FullColor
      mirroring=NoMirror
      resourceName=../../../Engine/Content/Slate/Common/ProgressBar_Marquee.png
      resourceObject=null
      tiling=Horizontal
      imageSize=<20 12>
      margin=<ltrb 0 0 0 0>
      outlineSettings {
       bUseBrushTransparency=false
       roundingType=HalfHeightRadius
       width=0
       color {colorUseRule=UseColor_Specified, specifiedColor=<rgba 0 0 0 0>}
       cornerRadii=<0 0 0 0>
      }
      tintColor {colorUseRule=UseColor_Specified, specifiedColor=<rgba 1 1 1 1>}
      uVRegion {bIsValid=false, max=<0 0>, min=<0 0>}
     }
    }
   }
   slot {
    class=CanvasPanelSlot
    properties {layoutData {alignment=<0.5 0.5>, anchors {maximum=<0.5 0>, minimum=<0.5 0>}, offsets=<ltrb 0 180 1000 20>}}
   }
  }
 ]
}
