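; Blueprint interface for PistolCanary's comic panel UI, declaring the calls to drive intro and hero panel presentation, boss health bars and a dev widget.
; Animation calls cover border fades, intro wipe, hero intro comic and panels chosen by hero, side, direction, play mode and speed.
; Explicit texture load and unload calls, including per-panel intro unloads, let callers manage texture residency.
; It holds only signatures, with no logic.
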(fn ComicBorderFadeIn)

(fn ComicBorderFadeOut)

(fn ComicIntroLoadTextures)

(fn ComicIntroUnloadTexturesPanel1)

(fn ComicIntroUnloadTexturesPanel2)

(fn ComicIntroUnloadTexturesPanel3)

(fn HeroPanel_LoadTextures)

(fn HeroPanel_UnloadTextures)

(fn PlayPistolCanaryHeroIntroComic)

(fn PlayPistolCanaryHeroPanelAnimation
  (params (HeroPanel :type E_PistolCanary_HeroPanelSelect%20Enum :default "NewEnumerator0") (PanelSide :type E_PistolCanary_PanelSide%20Enum :default "NewEnumerator0")))

(fn PlayPistolCanaryIntroWipeAnimation)

(fn PlayPistolCanaryPanelAnimation
  (params (HeroPanelSelect :type E_PistolCanary_HeroPanelSelect%20Enum :default "NewEnumerator0") (HeroPanelIndex :type Integer) (PanelSide :type E_PistolCanary_PanelSide%20Enum :default "NewEnumerator0") (PlayDirection :type EUMGSequencePlayMode%20Enum :default "Forward") (AnimationDirection :type E_PistolCanary_PanelDirection%20Enum :default "NewEnumerator0") (PlaySpeed :type Float%20%28double-precision%29 :default 1.0)))

(fn SetPistolCanaryBossHealthBarVisibility
  (params (HealthBarIndex :type Integer) (Visibility :type ESlateVisibility%20Enum :default "Visible")))

(fn UpdatePistolCanaryBossHealthBar
  (params (HealthBarIndex :type Integer) (Percent :type Float%20%28double-precision%29)))

(fn UpdatePistolCanaryDevWidget
  (params (SegmentA :type E_PistolCanary_DevSegmentA%20Enum :default "NewEnumerator3") (SegmentB :type E_PistolCanary_DevSegmentB%20Enum :default "NewEnumerator3") (SegmentC :type E_PistolCanary_DevSegmentC%20Enum :default "NewEnumerator11")))
