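# PistolCanary developer HUD widget that displays three segment states, a timer and a start indicator.
# UpdatePistolCanaryDevWidget converts its SegmentA, SegmentB and SegmentC enums to strings for three stacked bottom-center text blocks.
# UpdateTimer writes supplied text to the time readout beneath them.
# ShowStartSignal and HideStartSignal toggle a red circle anchored top-right, which PreConstruct hides initially.
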WidgetBlueprint /PistolCanary/Layout/BP/ComicPanels/PistolCanary_DevHUD
graphs [EventGraph {nodes [
   HideStartSignal() {is=title K2Node_CustomEvent}
   n1 = Get StartSignal() -> (StartSignal:Image Object Reference)
   > SetVisibility(self=@n1, InVisibility=Collapsed)
   n2 = UpdateTimer!() -> (OutputDelegate, then, InText:Text) {is=title K2Node_CustomEvent}
   > "SetText (Text)"(self=Get Text_Time() -> (Text_Time:Common Text Object Reference), InText=@n2.InText)
   ShowStartSignal() {is=title K2Node_CustomEvent}
   > SetVisibility(self=@n1, InVisibility=HitTestInvisible)
   Event PreConstruct()
   > n3 = HideStartSignal() {is=title K2Node_CallFunction}
   n4 = Sequence!(execute=[@n3.then, @n5.then]) -> (then_0, then_1, then_2:Exec)
   n5 = Event UpdatePistolCanaryDevWidget!() -> (
    OutputDelegate
    then
    SegmentA:E_PistolCanary_DevSegmentA Enum=NewEnumerator6
    SegmentB:E_PistolCanary_DevSegmentB Enum=NewEnumerator11
    SegmentC:E_PistolCanary_DevSegmentC Enum=NewEnumerator11
   ) {is=title K2Node_Event}
   "SetText (Text)"(
    execute=@n4.then_0
    self=Get TextA() -> (TextA:Common Text Object Reference)
    InText="To Text (String)"(InString=Enum to String!(Enumerator:E_PistolCanary_DevSegmentA Enum=@n5.SegmentA) -> (ReturnValue:String))
   )
   "SetText (Text)"(
    execute=@n4.then_1
    self=Get TextB() -> (TextB:Common Text Object Reference)
    InText="To Text (String)"(InString=Enum to String(Enumerator:E_PistolCanary_DevSegmentB Enum=@n5.SegmentB))
   )
   "SetText (Text)"(
    execute=@n4.then_2
    self=Get TextC() -> (TextC:Common Text Object Reference)
    InText="To Text (String)"(InString=Enum to String(Enumerator:E_PistolCanary_DevSegmentC Enum=@n5.SegmentC))
   )
  ]}]
variables []
widgets {
 info {inheritedWidgetCount=0, parentClass=&UserWidget, rootWidgetClass=&CanvasPanel, widgetCount=6}
 widgets [
  CanvasPanel_21 {widgetClassPath=&CanvasPanel, widget=CanvasPanel_21, properties {visibility=HitTestInvisible}}
  TextA {
   widgetClassPath=&CommonTextBlock
   bIsVariable=true
   parent=CanvasPanel_21
   widget=TextA
   properties {
    justification=Center
    style=&TextStyle-Base_C
    text=developer text here
    visibility=HitTestInvisible
    font {
     bForceMonospaced=false
     bMaterialIsStencil=false
     fontMaterial=null
     fontObject=&NotoSans
     letterSpacing=0
     monospacedWidth=1
     size=24
     skewAmount=0
     typefaceFontName=NotoSans-Regular
     outlineSettings {
      bApplyOutlineToDropShadows=false
      bMiteredCorners=false
      bSeparateFillAlpha=false
      outlineMaterial=null
      outlineSize=0
      outlineColor=<rgba 0 0 0 1>
     }
    }
    shadowOffset=<0 0>
   }
   slot {
    class=CanvasPanelSlot
    properties {bAutoSize=true, layoutData {alignment=<0.5 0.5>, anchors {maximum=<0.5 1>, minimum=<0.5 1>}, offsets=<ltrb 0 -154.2 100 30>}}
   }
  }
  TextB {
   widgetClassPath=&CommonTextBlock
   bIsVariable=true
   parent=CanvasPanel_21
   widget=TextB
   properties {
    justification=Center
    style=/FortUILibrary/Styles/Typography/Foundation/TextStyle-Base-XS.TextStyle-Base-XS_C
    text=developer text here
    visibility=HitTestInvisible
    font {
     bForceMonospaced=false
     bMaterialIsStencil=false
     fontMaterial=null
     fontObject=&NotoSans
     letterSpacing=0
     monospacedWidth=1
     size=16
     skewAmount=0
     typefaceFontName=NotoSans-Regular
     outlineSettings {
      bApplyOutlineToDropShadows=false
      bMiteredCorners=false
      bSeparateFillAlpha=false
      outlineMaterial=null
      outlineSize=0
      outlineColor=<rgba 0 0 0 1>
     }
    }
    shadowOffset=<0 0>
   }
   slot {
    class=CanvasPanelSlot
    properties {bAutoSize=true, layoutData {alignment=<0.5 0.5>, anchors {maximum=<0.5 1>, minimum=<0.5 1>}, offsets=<ltrb 0 -113.5 100 30>}}
   }
  }
  TextC {
   widgetClassPath=&CommonTextBlock
   bIsVariable=true
   parent=CanvasPanel_21
   widget=TextC
   properties {
    justification=Center
    style=&TextStyle-Base-XS_C
    text=developer text here
    visibility=HitTestInvisible
    font {
     bForceMonospaced=false
     bMaterialIsStencil=false
     fontMaterial=null
     fontObject=&NotoSans
     letterSpacing=0
     monospacedWidth=1
     size=16
     skewAmount=0
     typefaceFontName=NotoSans-Regular
     outlineSettings {
      bApplyOutlineToDropShadows=false
      bMiteredCorners=false
      bSeparateFillAlpha=false
      outlineMaterial=null
      outlineSize=0
      outlineColor=<rgba 0 0 0 1>
     }
    }
    shadowOffset=<0 0>
   }
   slot {
    class=CanvasPanelSlot
    properties {bAutoSize=true, layoutData {alignment=<0.5 0.5>, anchors {maximum=<0.5 1>, minimum=<0.5 1>}, offsets=<ltrb 0 -77.5 100 30>}}
   }
  }
  Text_Time {
   widgetClassPath=&CommonTextBlock
   bIsVariable=true
   parent=CanvasPanel_21
   widget=Text_Time
   properties {
    justification=Center
    style=/FortUILibrary/Styles/Typography/Foundation/TextStyle-Base-M.TextStyle-Base-M_C
    text=0
    visibility=HitTestInvisible
    font {
     bForceMonospaced=false
     bMaterialIsStencil=false
     fontMaterial=null
     fontObject=&NotoSans
     letterSpacing=0
     monospacedWidth=1
     size=24
     skewAmount=0
     typefaceFontName=NotoSans-Regular
     outlineSettings {
      bApplyOutlineToDropShadows=false
      bMiteredCorners=false
      bSeparateFillAlpha=false
      outlineMaterial=null
      outlineSize=0
      outlineColor=<rgba 0 0 0 1>
     }
    }
    shadowOffset=<0 0>
   }
   slot {
    class=CanvasPanelSlot
    properties {layoutData {alignment=<0.5 0.5>, anchors {maximum=<0.5 1>, minimum=<0.5 1>}, offsets=<ltrb 0 -25 400 50>}}
   }
  }
  StartSignal {
   widgetClassPath=&Image
   bIsVariable=true
   parent=CanvasPanel_21
   widget=StartSignal
   properties {brush {
     bIsDynamicallyLoaded=false
     drawAs=Image
     imageType=FullColor
     mirroring=NoMirror
     resourceName=None
     resourceObject=/LauncherSocial/circle
     tiling=NoTile
     imageSize=<24 24>
     margin=<ltrb 0 0 0 0>
     outlineSettings {
      bUseBrushTransparency=false
      roundingType=HalfHeightRadius
      width=0
      color {colorUseRule=UseColor_Specified, specifiedColor=<rgba 0 0 0 0>}
      cornerRadii=<0 0 0 0>
     }
     tintColor {colorUseRule=UseColor_Specified, specifiedColor=<rgba 1 0 0 1>}
     uVRegion {bIsValid=false, max=<0 0>, min=<0 0>}
    }}
   slot {class=CanvasPanelSlot, properties {layoutData {alignment=<0 0>, anchors {maximum=<1 0>, minimum=<1 0>}, offsets=<ltrb -200 0 200 200>}}}
  }
 ]
}
