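# Full-screen comic panel widget for Pistol Canary that wipes away during the intro.
# A single click-through Image is the root, with a white-tinted brush and no texture assigned.
# The PlayPistolCanaryIntroWipeAnimation event plays PanelWipe, a quick animation that slides the image off to the left while fading it out.
# The graph has no other logic, as PreConstruct, Construct and Tick are empty.
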WidgetBlueprint /PistolCanary/Layout/BP/ComicPanels/PistolCanary_FullScreenPanel
animations [PanelWipe {
  binding_count=1
  length_seconds=0.25
  object=&self:PanelWipe
  sequence {
   bindings [Image_33 {
     id=#1
     possessed_class=Image
     tracks [
      {
       type=MovieScene2DTransformTrack
       sections [{
         type=MovieScene2DTransformSection
         end_frame=15
         start_frame=0
         channels [
          Translation.X {type=MovieSceneScriptingFloatChannel, keys [[0,0],[15,-2199.7319]]}
          Translation.Y {type=MovieSceneScriptingFloatChannel, const [0,1]}
          Angle {type=MovieSceneScriptingFloatChannel, const [0,1]}
          Scale.X {type=MovieSceneScriptingFloatChannel, const [1,1]}
          Scale.Y {type=MovieSceneScriptingFloatChannel, const [1,1]}
          Shear.X {type=MovieSceneScriptingFloatChannel, const [0,1]}
          Shear.Y {type=MovieSceneScriptingFloatChannel, const [0,1]}
         ]
        }]
      }
      {
       type=MovieSceneFloatTrack
       display_name=Render Opacity
       sections [{type=MovieSceneFloatSection, end_frame=15, start_frame=0, channels [None {type=MovieSceneScriptingFloatChannel, keys [[0,1],[14.999,0]]}]}]
      }
     ]
    }]
   display_rate {numerator=60}
  }
 }]
graphs [EventGraph {nodes [
   Event PlayPistolCanaryIntroWipeAnimation() {is=title K2Node_Event}
   > PlayAnimation(InAnimation=Get PanelWipe() -> (PanelWipe:Widget Animation Object Reference))
   Event PreConstruct()
   Event Construct()
   Event Tick()
  ]}]
variables []
widgets {
 info {inheritedWidgetCount=0, parentClass=&UserWidget, rootWidgetClass=&Image, widgetCount=1}
 widgets [Image_33 {
   widgetClassPath=&Image
   bIsVariable=true
   widget=Image_33
   properties {
    visibility=HitTestInvisible
    brush {
     bIsDynamicallyLoaded=false
     drawAs=Image
     imageType=FullColor
     mirroring=NoMirror
     resourceName=None
     resourceObject=null
     tiling=NoTile
     imageSize=<32 32>
     margin=<ltrb 0 0 0 0>
     outlineSettings {
      bUseBrushTransparency=false
      roundingType=HalfHeightRadius
      width=0
      color {colorUseRule=UseColor_Specified, specifiedColor=<rgba 0 0 0 0>}
      cornerRadii=<0 0 0 0>
     }
     tintColor {colorUseRule=UseColor_Specified, specifiedColor=<rgba 1 1 1 1>}
     uVRegion {bIsValid=false, max=<0 0>, min=<0 0>}
    }
   }
  }]
}
