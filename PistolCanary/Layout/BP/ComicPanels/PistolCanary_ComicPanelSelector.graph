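# Widget Blueprint for a single-image comic panel, meant to show a PistolCanary hero overlay chosen by index.
# SetPanelImageByIndexAsync selects one of eight soft texture slots (A, B, D variants; slots 4 and 5 repeat A1), async-loads it and applies it at matched size.
# PreConstruct instead passes ImageIndex to SetPanelImageByIndex, which ignores the index and sets the brush without a texture.
# Nothing here calls the async event.
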WidgetBlueprint /PistolCanary/Layout/BP/ComicPanels/PistolCanary_ComicPanelSelector
graphs [
 EventGraph {nodes [
   n1 = SetPanelImageByIndexAsync!() -> (OutputDelegate, then, Index:int) {is=title K2Node_CustomEvent}
   > n2 = Async Load Asset(Asset=Select!(
     Option 0:Soft Object Reference=/PistolCanary/Layout/Texture/ComicPanels/HeroPanels/PC_Color_Comic_Overlay_A1
     Option 1:Soft Object Reference=/PistolCanary/Layout/Texture/ComicPanels/HeroPanels/PC_Color_Comic_Overlay_A2
     Option 2:Soft Object Reference=/PistolCanary/Layout/Texture/ComicPanels/HeroPanels/PC_Color_Comic_Overlay_B1
     Option 3:Soft Object Reference=/PistolCanary/Layout/Texture/ComicPanels/HeroPanels/PC_Color_Comic_Overlay_B2
     Option 4:Soft Object Reference=&PC_Color_Comic_Overlay_A1
     Option 5:Soft Object Reference=&PC_Color_Comic_Overlay_A1
     Option 6:Soft Object Reference=/PistolCanary/Layout/Texture/ComicPanels/HeroPanels/PC_Color_Comic_Overlay_D1
     Option 7:Soft Object Reference=/PistolCanary/Layout/Texture/ComicPanels/HeroPanels/PC_Color_Comic_Overlay_D2
     Index:int=@n1.Index
    ) -> (ReturnValue:Soft Object Reference))
   > SetBrushFromTexture!(
    execute
    self:Image Object Reference=Get Image_11() -> (Image_11:Image Object Reference)
    Texture:Texture 2D Object Reference=Cast To Texture2D!(Object=@n2) -> (AsTexture 2D, bSuccess:bool=~)
    bMatchSize:bool=true
   ) -> (then)
   Event PreConstruct()
   > SetPanelImageByIndex!(execute, self, Index:int=Get ImageIndex() -> (ImageIndex:int=0)) -> (then)
   Event Construct()
   Event Tick()
  ]}
 SetPanelImageByIndex {nodes [
   SetPanelImageByIndex!() -> (then, Index:int) {is=title K2Node_FunctionEntry}
   > SetBrushFromTexture(self=Get Image_11(), bMatchSize=true)
  ]}
]
variables [ImageIndex {type=integer}]
widgets {
 info {inheritedWidgetCount=0, parentClass=&UserWidget, rootWidgetClass=&Image, widgetCount=1}
 widgets [Image_11 {
   widgetClassPath=&Image
   bIsVariable=true
   widget=Image_11
   properties {
    pixelSnapping=Disabled
    visibility=HitTestInvisible
    brush {
     bIsDynamicallyLoaded=false
     drawAs=Image
     imageType=FullColor
     mirroring=NoMirror
     resourceName=None
     resourceObject=null
     tiling=NoTile
     imageSize=<168 256>
     margin=<ltrb 0 0 0 0>
     outlineSettings {
      bUseBrushTransparency=false
      roundingType=HalfHeightRadius
      width=0
      color {colorUseRule=UseColor_Specified, specifiedColor=<rgba 0 0 0 0>}
      cornerRadii=<0 0 0 0>
     }
     tintColor {colorUseRule=UseColor_Specified, specifiedColor=<rgba 1 1 1 1>}
     uVRegion {bIsValid=false, max=<0 0>, min=<0 0>}
    }
   }
  }]
}
