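# PistolCanary UMG layout widget that animates left and right edge-anchored comic panels sliding and fading on screen, both hidden on Construct.
# PlayPistolCanaryPanelAnimation sets both panel images from HeroPanelIndex, then plays horizontal or vertical animations by side and direction; reverse playback skips the image update.
# PlayPistolCanaryHeroPanelAnimation maps a hero enum to one panel, image and vertical animation variant.
# Most animations bake slide-in, hold and slide-out into one timeline.
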WidgetBlueprint /PistolCanary/Layout/BP/ComicPanels/PistolCanary_PanelLayout
animations [
 PanelLeftSlideInHorizontal {
  binding_count=1
  length_seconds=2.3167
  object=&self:PanelLeftSlideInHorizontal
  sequence {
   bindings [ComicPanel_Left {
     id=#1
     possessed_class=PistolCanary_ComicPanelSelector_C
     tracks [
      {
       type=MovieScene2DTransformTrack
       sections [{
         type=MovieScene2DTransformSection
         end_frame=139
         start_frame=0
         channels [
          Translation.X {type=MovieSceneScriptingFloatChannel, keys [[0,-684.1262],[15,0],[123,0],[138,-656.1862]]}
          Translation.Y {type=MovieSceneScriptingFloatChannel, const [0,2]}
          Angle {type=MovieSceneScriptingFloatChannel, const [0,2]}
          Scale.X {type=MovieSceneScriptingFloatChannel, const [1,2]}
          Scale.Y {type=MovieSceneScriptingFloatChannel, const [1,2]}
          Shear.X {type=MovieSceneScriptingFloatChannel, const [0,2]}
          Shear.Y {type=MovieSceneScriptingFloatChannel, const [0,2]}
         ]
        }]
      }
      {
       type=MovieSceneFloatTrack
       display_name=Render Opacity
       sections [{
         type=MovieSceneFloatSection
         end_frame=139
         start_frame=0
         channels [None {type=MovieSceneScriptingFloatChannel, keys [[0,0],[15,1],[123,1],[138,0]]}]
        }]
      }
     ]
    }]
   display_rate {numerator=60}
  }
 }
 PanelRightSlideInHorizontal {
  binding_count=1
  length_seconds=2.3167
  object=&self:PanelRightSlideInHorizontal
  sequence {
   bindings [ComicPanel_Right {
     id=#2
     possessed_class=PistolCanary_ComicPanelSelector_C
     tracks [
      {
       type=MovieScene2DTransformTrack
       sections [{
         type=MovieScene2DTransformSection
         end_frame=139
         start_frame=0
         channels [
          Translation.X {type=MovieSceneScriptingFloatChannel, keys [[0,681.0323],[15,0],[123,0],[138,673.1769]]}
          Translation.Y {type=MovieSceneScriptingFloatChannel, const [0,2]}
          Angle {type=MovieSceneScriptingFloatChannel, const [0,2]}
          Scale.X {type=MovieSceneScriptingFloatChannel, const [1,2]}
          Scale.Y {type=MovieSceneScriptingFloatChannel, const [1,2]}
          Shear.X {type=MovieSceneScriptingFloatChannel, const [0,2]}
          Shear.Y {type=MovieSceneScriptingFloatChannel, const [0,2]}
         ]
        }]
      }
      {
       type=MovieSceneFloatTrack
       display_name=Render Opacity
       sections [{
         type=MovieSceneFloatSection
         end_frame=139
         start_frame=0
         channels [None {type=MovieSceneScriptingFloatChannel, keys [[0,0],[15,1],[123,1],[138,0]]}]
        }]
      }
     ]
    }]
   display_rate {numerator=60}
  }
 }
 PanelLeftSlideInVertical {
  binding_count=1
  length_seconds=7.6333
  object=&self:PanelLeftSlideInVertical
  sequence {
   bindings [ComicPanel_Left {
     id=#1
     possessed_class=PistolCanary_ComicPanelSelector_C
     tracks [
      {
       type=MovieScene2DTransformTrack
       sections [{
         type=MovieScene2DTransformSection
         end_frame=458
         start_frame=0
         channels [
          Translation.X {type=MovieSceneScriptingFloatChannel, thinned_from=4, keys [[0,-308.7808],[16,0],[457,0]]}
          Translation.Y {type=MovieSceneScriptingFloatChannel, keys [[0,1088.9025],[16,0],[442,0],[457,-1091.6603]]}
          Angle {type=MovieSceneScriptingFloatChannel, const [0,2]}
          Scale.X {type=MovieSceneScriptingFloatChannel, const [1,2]}
          Scale.Y {type=MovieSceneScriptingFloatChannel, const [1,2]}
          Shear.X {type=MovieSceneScriptingFloatChannel, const [0,2]}
          Shear.Y {type=MovieSceneScriptingFloatChannel, const [0,2]}
         ]
        }]
      }
      {
       type=MovieSceneFloatTrack
       display_name=Render Opacity
       sections [{
         type=MovieSceneFloatSection
         end_frame=458
         start_frame=0
         channels [None {type=MovieSceneScriptingFloatChannel, keys [[0,0],[16,1],[442,1],[457,0]]}]
        }]
      }
     ]
    }]
   display_rate {numerator=60}
  }
 }
 PanelRightSlideInVertical {
  binding_count=1
  length_seconds=6.2333
  object=&self:PanelRightSlideInVertical
  sequence {
   bindings [ComicPanel_Right {
     id=#2
     possessed_class=PistolCanary_ComicPanelSelector_C
     tracks [
      {
       type=MovieScene2DTransformTrack
       sections [{
         type=MovieScene2DTransformSection
         end_frame=375
         start_frame=1
         channels [
          Translation.X {type=MovieSceneScriptingFloatChannel, thinned_from=4, keys [[1,353.58],[16,0],[374,0]]}
          Translation.Y {type=MovieSceneScriptingFloatChannel, keys [[1,-1081.0922],[16,0],[358,0],[374,1082.3345]]}
          Angle {type=MovieSceneScriptingFloatChannel, const [0,2]}
          Scale.X {type=MovieSceneScriptingFloatChannel, const [1,2]}
          Scale.Y {type=MovieSceneScriptingFloatChannel, const [1,2]}
          Shear.X {type=MovieSceneScriptingFloatChannel, const [0,2]}
          Shear.Y {type=MovieSceneScriptingFloatChannel, const [0,2]}
         ]
        }]
      }
      {
       type=MovieSceneFloatTrack
       display_name=Render Opacity
       sections [{
         type=MovieSceneFloatSection
         end_frame=375
         start_frame=1
         channels [None {type=MovieSceneScriptingFloatChannel, keys [[1,0],[16,1],[358,1],[374,0]]}]
        }]
      }
     ]
    }]
   display_rate {numerator=60}
  }
 }
 PanelLeftSlideInVertical_2 {
  binding_count=1
  length_seconds=7.9667
  object=&self:PanelLeftSlideInVertical_2
  sequence {
   bindings [ComicPanel_Left {
     id=#1
     possessed_class=PistolCanary_ComicPanelSelector_C
     tracks [
      {
       type=MovieScene2DTransformTrack
       sections [{
         type=MovieScene2DTransformSection
         end_frame=478
         start_frame=0
         channels [
          Translation.X {type=MovieSceneScriptingFloatChannel, thinned_from=4, keys [[0,-341.2048],[16,0],[477,0]]}
          Translation.Y {type=MovieSceneScriptingFloatChannel, keys [[0,-1082.9917],[16,0],[462,0],[477,1085.1641]]}
          Angle {type=MovieSceneScriptingFloatChannel, const [0,3]}
          Scale.X {type=MovieSceneScriptingFloatChannel, const [1,3]}
          Scale.Y {type=MovieSceneScriptingFloatChannel, const [1,3]}
          Shear.X {type=MovieSceneScriptingFloatChannel, const [0,3]}
          Shear.Y {type=MovieSceneScriptingFloatChannel, const [0,3]}
         ]
        }]
      }
      {
       type=MovieSceneFloatTrack
       display_name=Render Opacity
       sections [{
         type=MovieSceneFloatSection
         end_frame=478
         start_frame=0
         channels [None {type=MovieSceneScriptingFloatChannel, thinned_from=4, keys [[0,1],[462,1],[477,0]]}]
        }]
      }
     ]
    }]
   display_rate {numerator=60}
  }
 }
 PanelRightSlideInVertical_2 {
  binding_count=1
  length_seconds=6.5667
  object=&self:PanelRightSlideInVertical_2
  sequence {
   bindings [ComicPanel_Right {
     id=#2
     possessed_class=PistolCanary_ComicPanelSelector_C
     tracks [
      {
       type=MovieScene2DTransformTrack
       sections [{
         type=MovieScene2DTransformSection
         end_frame=395
         start_frame=1
         channels [
          Translation.X {type=MovieSceneScriptingFloatChannel, thinned_from=4, keys [[1,353.58],[16,0],[394,0]]}
          Translation.Y {type=MovieSceneScriptingFloatChannel, keys [[1,1081.635],[16,0],[378,0],[394,-1083.9471]]}
          Angle {type=MovieSceneScriptingFloatChannel, const [0,4]}
          Scale.X {type=MovieSceneScriptingFloatChannel, const [1,4]}
          Scale.Y {type=MovieSceneScriptingFloatChannel, const [1,4]}
          Shear.X {type=MovieSceneScriptingFloatChannel, const [0,4]}
          Shear.Y {type=MovieSceneScriptingFloatChannel, const [0,4]}
         ]
        }]
      }
      {
       type=MovieSceneFloatTrack
       display_name=Render Opacity
       sections [{
         type=MovieSceneFloatSection
         end_frame=395
         start_frame=1
         channels [None {type=MovieSceneScriptingFloatChannel, keys [[1,0],[16,1],[378,1],[394,0]]}]
        }]
      }
     ]
    }]
   display_rate {numerator=60}
  }
 }
 PanelLeftSlideInVertical_3 {
  binding_count=1
  length_seconds=9.45
  object=&self:PanelLeftSlideInVertical_3
  sequence {
   bindings [ComicPanel_Left {
     id=#1
     possessed_class=PistolCanary_ComicPanelSelector_C
     tracks [
      {
       type=MovieScene2DTransformTrack
       sections [{
         type=MovieScene2DTransformSection
         end_frame=567
         start_frame=0
         channels [
          Translation.X {type=MovieSceneScriptingFloatChannel, thinned_from=4, keys [[0,-308.7808],[16,0],[566,0]]}
          Translation.Y {type=MovieSceneScriptingFloatChannel, keys [[0,1088.9025],[16,0],[552,0],[566,-1091.6603]]}
          Angle {type=MovieSceneScriptingFloatChannel, const [0,2]}
          Scale.X {type=MovieSceneScriptingFloatChannel, const [1,2]}
          Scale.Y {type=MovieSceneScriptingFloatChannel, const [1,2]}
          Shear.X {type=MovieSceneScriptingFloatChannel, const [0,2]}
          Shear.Y {type=MovieSceneScriptingFloatChannel, const [0,2]}
         ]
        }]
      }
      {
       type=MovieSceneFloatTrack
       display_name=Render Opacity
       sections [{
         type=MovieSceneFloatSection
         end_frame=567
         start_frame=0
         channels [None {type=MovieSceneScriptingFloatChannel, keys [[0,0],[16,1],[552,1],[566,0]]}]
        }]
      }
     ]
    }]
   display_rate {numerator=60}
  }
 }
 PanelRightSlideInVertical_3 {
  binding_count=1
  length_seconds=8.05
  object=&self:PanelRightSlideInVertical_3
  sequence {
   bindings [ComicPanel_Right {
     id=#2
     possessed_class=PistolCanary_ComicPanelSelector_C
     tracks [
      {
       type=MovieScene2DTransformTrack
       sections [{
         type=MovieScene2DTransformSection
         end_frame=484
         start_frame=1
         channels [
          Translation.X {type=MovieSceneScriptingFloatChannel, thinned_from=4, keys [[1,353.58],[16,0],[483,0]]}
          Translation.Y {type=MovieSceneScriptingFloatChannel, keys [[1,-1081.0922],[16,0],[467,0],[483,1082.3345]]}
          Angle {type=MovieSceneScriptingFloatChannel, const [0,2]}
          Scale.X {type=MovieSceneScriptingFloatChannel, const [1,2]}
          Scale.Y {type=MovieSceneScriptingFloatChannel, const [1,2]}
          Shear.X {type=MovieSceneScriptingFloatChannel, const [0,2]}
          Shear.Y {type=MovieSceneScriptingFloatChannel, const [0,2]}
         ]
        }]
      }
      {
       type=MovieSceneFloatTrack
       display_name=Render Opacity
       sections [{
         type=MovieSceneFloatSection
         end_frame=484
         start_frame=1
         channels [None {type=MovieSceneScriptingFloatChannel, keys [[1,0],[16,1],[467,1],[483,0]]}]
        }]
      }
     ]
    }]
   display_rate {numerator=60}
  }
 }
]
graphs [EventGraph {nodes [
   n1 = Get ComicPanel_Right() -> (ComicPanel_Right:Pistol Canary Comic Panel Selector Object Reference)
   n2 = Event PlayPistolCanaryPanelAnimation!() -> (
    OutputDelegate
    then
    HeroPanelSelect:E_PistolCanary_HeroPanelSelect Enum=NewEnumerator0
    HeroPanelIndex:int=0
    PanelSide:E_PistolCanary_PanelSide Enum=NewEnumerator0
    PlayDirection:EUMGSequencePlayMode Enum=Forward
    AnimationDirection:E_PistolCanary_PanelDirection Enum=NewEnumerator0
    PlaySpeed:double
   ) {is=title K2Node_Event}
   n3 = Set ImageIndex!(
    execute=[@n11.then, @n15.then]
    ImageIndex:int=Select!(Option 0:int=1, Option 1:int=3, Option 2:int=5, Option 3:int=7, Index:int=@n2.HeroPanelIndex) -> (ReturnValue:int)
    self:Pistol Canary Comic Panel Selector Object Reference=@n1
   ) -> (then, Output_Get:int=0)
   n4 = SetPanelImageByIndexAsync!(execute, self:Pistol Canary Comic Panel Selector Object Reference=@n1, Index:int=@n3) -> (then)
   n5 = Switch on E_PistolCanary_PanelSide!(execute=[@n4.then, @n16.then], Selection:E_PistolCanary_PanelSide Enum=@n2.PanelSide) -> (NewEnumerator0:Exec, NewEnumerator1:Exec, NewEnumerator2:Exec) {is=title K2Node_SwitchEnum}
   n6 = Sequence(execute=@n5.NewEnumerator2)
   n7 = Select!(
    NewEnumerator0:Widget Animation Object Reference=Get PanelLeftSlideInHorizontal() -> (PanelLeftSlideInHorizontal:Widget Animation Object Reference)
    NewEnumerator1:Widget Animation Object Reference=Get PanelLeftSlideInVertical() -> (PanelLeftSlideInVertical:Widget Animation Object Reference)
    Index:E_PistolCanary_PanelDirection Enum=@n2.AnimationDirection
   ) -> (ReturnValue:Widget Animation Object Reference)
   PlayAnimation(execute=[@n6.then_0, @n5.NewEnumerator0], InAnimation=@n7, PlaybackSpeed=@n2.PlaySpeed)
   n8 = Select(
    NewEnumerator0=Get PanelRightSlideInHorizontal() -> (PanelRightSlideInHorizontal:Widget Animation Object Reference)
    NewEnumerator1=Get PanelRightSlideInVertical() -> (PanelRightSlideInVertical:Widget Animation Object Reference)
    Index=@n2.AnimationDirection
   )
   PlayAnimation(execute=[@n6.then_1, @n5.NewEnumerator1], InAnimation=@n8, PlaybackSpeed=@n2.PlaySpeed)
   n9 = Get ComicPanel_Left() -> (ComicPanel_Left:Pistol Canary Comic Panel Selector Object Reference)
   n10 = Set ImageIndex(
    execute=@n18.Forward
    ImageIndex=Select!(Option 0:int=0, Option 1:int=2, Option 2:int=4, Option 3:int=6, Index:int=@n2.HeroPanelIndex) -> (ReturnValue:int)
    self=@n9
   ) -> (Output_Get=0)
   n11 = SetPanelImageByIndexAsync(self=@n9, Index=@n10)
   Event Construct()
   > SetRenderOpacity(self=[Get ComicPanel_Left(), Get ComicPanel_Right()], InOpacity=0)
   n12 = Event PlayPistolCanaryHeroPanelAnimation!() -> (OutputDelegate, then, HeroPanel:E_PistolCanary_HeroPanelSelect Enum=NewEnumerator0, PanelSide:E_PistolCanary_PanelSide Enum=NewEnumerator0) {is=title K2Node_Event}
   n13 = Get ComicPanel_Right()
   n14 = Get ComicPanel_Left()
   > SetPanelImageByIndex!(
    execute
    self:Pistol Canary Comic Panel Selector Object Reference=Select!(
     NewEnumerator0:Pistol Canary Comic Panel Selector Object Reference=@n14
     NewEnumerator1:Pistol Canary Comic Panel Selector Object Reference=@n13
     NewEnumerator2:Pistol Canary Comic Panel Selector Object Reference=@n14
     NewEnumerator3:Pistol Canary Comic Panel Selector Object Reference=@n13
     NewEnumerator4:Pistol Canary Comic Panel Selector Object Reference=@n14
     NewEnumerator5:Pistol Canary Comic Panel Selector Object Reference=@n13
     Index:E_PistolCanary_HeroPanelSelect Enum=@n12.HeroPanel
    ) -> (ReturnValue:Pistol Canary Comic Panel Selector Object Reference)
    Index:int=Select(
     NewEnumerator0:int=0
     NewEnumerator1:int=1
     NewEnumerator2:int=2
     NewEnumerator3:int=3
     NewEnumerator4:int=6
     NewEnumerator5:int=7
     Index=@n12.HeroPanel
    ) -> (ReturnValue:int)
   ) -> (then)
   > PlayAnimation(InAnimation=Select(
     NewEnumerator0:Widget Animation Object Reference=Get PanelLeftSlideInVertical()
     NewEnumerator1:Widget Animation Object Reference=Get PanelRightSlideInVertical()
     NewEnumerator2:Widget Animation Object Reference=Get PanelLeftSlideInVertical_2() -> (PanelLeftSlideInVertical_2:Widget Animation Object Reference)
     NewEnumerator3:Widget Animation Object Reference=Get PanelRightSlideInVertical_2() -> (PanelRightSlideInVertical_2:Widget Animation Object Reference)
     NewEnumerator4:Widget Animation Object Reference=Get PanelLeftSlideInVertical_3() -> (PanelLeftSlideInVertical_3:Widget Animation Object Reference)
     NewEnumerator5:Widget Animation Object Reference=Get PanelRightSlideInVertical_3() -> (PanelRightSlideInVertical_3:Widget Animation Object Reference)
     Index=@n12.HeroPanel
    ) -> (ReturnValue:Widget Animation Object Reference))
   PlayAnimationReverse(execute=[@n17.then_1, @n19.NewEnumerator1], InAnimation=@n8, PlaybackSpeed=@n2.PlaySpeed)
   PlayAnimationReverse(execute=[@n17.then_0, @n19.NewEnumerator0], InAnimation=@n7, PlaybackSpeed=@n2.PlaySpeed)
   Event PreConstruct()
   Event Tick()
   n15 = SetPanelImageByIndex(execute=@n10.then, self=@n9, Index=@n10)
   n16 = SetPanelImageByIndex(execute=@n3.then, self=@n1, Index=@n3)
   n17 = Sequence(execute=@n19.NewEnumerator2)
   n18 = Switch on EUMGSequencePlayMode!(execute=@n2.then, Selection:EUMGSequencePlayMode Enum=@n2.PlayDirection) -> (Forward:Exec, Reverse:Exec, PingPong:Exec) {is=title K2Node_SwitchEnum}
   n19 = Switch on E_PistolCanary_PanelSide(execute=@n18.Reverse, Selection=@n2.PanelSide)
  ]}]
variables []
widgets {
 info {inheritedWidgetCount=0, parentClass=&UserWidget, rootWidgetClass=&CanvasPanel, widgetCount=3}
 widgets [
  CanvasPanel_163 {widgetClassPath=&CanvasPanel, widget=CanvasPanel_163, properties {visibility=HitTestInvisible}}
  ComicPanel_Left {
   widgetClassPath=&/PistolCanary_ComicPanelSelector
   bIsVariable=true
   parent=CanvasPanel_163
   widget=ComicPanel_Left
   properties {pixelSnapping=Disabled, visibility=HitTestInvisible}
   slot {class=CanvasPanelSlot, properties {layoutData {alignment=<0 0>, anchors {maximum=<0 1>, minimum=<0 0>}, offsets=<ltrb 0 0 672 0>}}}
  }
  ComicPanel_Right {
   widgetClassPath=&PistolCanary_ComicPanelSelector
   bIsVariable=true
   parent=CanvasPanel_163
   widget=ComicPanel_Right
   properties {imageIndex=1, pixelSnapping=Disabled, visibility=HitTestInvisible}
   slot {class=CanvasPanelSlot, properties {layoutData {alignment=<0 0>, anchors {maximum=<1 1>, minimum=<1 0>}, offsets=<ltrb -712 0 712 0>}}}
  }
 ]
}
