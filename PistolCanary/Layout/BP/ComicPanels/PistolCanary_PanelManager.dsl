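; PistolCanary panel manager that controls comic-style UI: hero intro, border, three hero panels and boss health bars.
; Custom events async-load widget classes, add them to the viewport and remove them, often forcing garbage collection.
; Interface messages and casts drive animations, bar updates and explicit texture load/unload; intro creation and playback skip the server.
; A looping timer feeds Level Sequence times to a dev HUD; several dev, movie and fullscreen-panel events are empty stubs.
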(event Custom|LoadComicBorderWidget
  ; comic border
  (bind _class (Utilities|AsyncLoadClassAsset "/PistolCanary/Layout/BP/ComicPanels/HeroPanels/PistolCanary_ComicBorder.PistolCanary_ComicBorder_C")
    (:Completed
      (bind _returnvalue (UserInterface|CreateWidget (Utilities|Casting|CastToUserWidgetClass _class)))
      (bind _output_get (Variables|Default|SetCurrentComicBorder _returnvalue))
      (UserInterface|Viewport|AddtoViewport _output_get 5))))

(event Custom|CreateHeroIntroWidget
  ; create hero intro widget
  (if (not (Networking|IsServer))
    (bind _class (Utilities|AsyncLoadClassAsset "/PistolCanary/Layout/BP/ComicPanels/HeroIntro/PistolCanary_ComicHeroIntro.PistolCanary_ComicHeroIntro_C")
      (:Completed
        (bind _returnvalue (UserInterface|CreateWidget (Utilities|Casting|CastToUserWidgetClass _class)))
        (Variables|Default|SetCurrentHeroIntroWidget _returnvalue)))))

(event Custom|ComicIntro_LoadTextures
  ; comic intro load textures
  (PistolCanary|ComicIntroLoadTextures(Message) (Variables|Default|GetCurrentHeroIntroWidget)))

(event Custom|RemoveComicBorderWidget
  ; comic border
  (bind _currentcomicborder (Variables|Default|GetCurrentComicBorder)
    (:then
      (Widget|RemovefromParent _currentcomicborder)
      (Variables|Default|SetCurrentComicBorder 0))
    (:else)))

(event Custom|ComicBorderFadeIn
  ; comic border
  (PistolCanary|ComicBorderFadeIn(Message) (Variables|Default|GetCurrentComicBorder)))

(event Custom|ComicIntro_UnloadTexturesPanel1
  ; comic intro unload textures
  (PistolCanary|ComicIntroUnloadTexturesPanel1(Message) (Variables|Default|GetCurrentHeroIntroWidget)))

(event Custom|AddAndPlayHeroIntro
  (bind _currentherointrowidget (Variables|Default|GetCurrentHeroIntroWidget))
  ; Add hero intro comic to viewport and play
  (if (not (Networking|IsServer))
    (UserInterface|Viewport|AddtoViewport _currentherointrowidget 1)
    (PistolCanary|PlayPistolCanaryHeroIntroComic(Message) _currentherointrowidget)))

(event Custom|ComicBorderFadeOut
  ; comic border
  (PistolCanary|ComicBorderFadeOut(Message) (Variables|Default|GetCurrentComicBorder)))

(event Custom|ComicIntro_UnloadTexturesPanel2
  ; comic intro unload textures
  (PistolCanary|ComicIntroUnloadTexturesPanel2(Message) (Variables|Default|GetCurrentHeroIntroWidget)))

(event Custom|RemoveHeroIntroWidget
  ; Add hero intro comic to viewport and play
  (bind _currentherointrowidget (Variables|Default|GetCurrentHeroIntroWidget)
    (:then
      (Widget|RemovefromParent _currentherointrowidget)
      (Variables|Default|SetCurrentHeroIntroWidget 0)
      (Utilities|Platform|CollectGarbage))
    (:else)))

(event Custom|ComicIntro_UnloadTexturesPanel3
  ; comic intro unload textures
  (PistolCanary|ComicIntroUnloadTexturesPanel3(Message) (Variables|Default|GetCurrentHeroIntroWidget)))

(event Custom|AddPanelWidget
  ; create hero panel widget
  (bind _class (Utilities|AsyncLoadClassAsset "/PistolCanary/Layout/BP/ComicPanels/PistolCanary_PanelLayout.PistolCanary_PanelLayout_C")
    (:Completed
      (bind _returnvalue (UserInterface|CreateWidget (Utilities|Casting|CastToUserWidgetClass _class)))
      (bind _output_get (Variables|Default|SetCurrentPanel _returnvalue))
      (UserInterface|Viewport|AddtoViewport _output_get 1))))

(event Custom|LoadHeroPanel1_Widget
  ; create hero panel 1 widget
  (bind _class (Utilities|AsyncLoadClassAsset "/PistolCanary/Layout/ComicPanels_v2/Blueprints/PistolCanary_HeroPanel1_v2.PistolCanary_HeroPanel1_v2_C")
    (:Completed
      (bind _returnvalue (UserInterface|CreateWidget (Utilities|Casting|CastToUserWidgetClass _class)))
      (bind _output_get (Variables|Default|SetCurrentHeroPanel1 _returnvalue))
      (UserInterface|Viewport|AddtoViewport _output_get 1))))

(event Custom|LoadHeroPanel2_Widget
  ; create hero panel 2 widget
  (bind _class (Utilities|AsyncLoadClassAsset "/PistolCanary/Layout/ComicPanels_v2/Blueprints/PistolCanary_HeroPanel2_v2.PistolCanary_HeroPanel2_v2_C")
    (:Completed
      (bind _returnvalue (UserInterface|CreateWidget (Utilities|Casting|CastToUserWidgetClass _class)))
      (bind _output_get (Variables|Default|SetCurrentHeroPanel2 _returnvalue))
      (UserInterface|Viewport|AddtoViewport _output_get 1))))

(event Custom|LoadHeroPanel3_Widget
  ; create hero panel 3 widget
  (bind _class (Utilities|AsyncLoadClassAsset "/PistolCanary/Layout/ComicPanels_v2/Blueprints/PistolCanary_HeroPanel3_v2.PistolCanary_HeroPanel3_v2_C")
    (:Completed
      (bind _returnvalue (UserInterface|CreateWidget (Utilities|Casting|CastToUserWidgetClass _class)))
      (bind _output_get (Variables|Default|SetCurrentHeroPanel3 _returnvalue))
      (UserInterface|Viewport|AddtoViewport _output_get 1))))

(event Custom|RemovePanelWidget
  ; create hero panel widget
  (bind _currentpanel (Variables|Default|GetCurrentPanel)
    (:then
      (Widget|RemovefromParent _currentpanel)
      (Variables|Default|SetCurrentPanel 0)
      (Utilities|Platform|CollectGarbage))
    (:else)))

(event Custom|RemovePanel1_Widget
  ; create hero panel 1 widget
  (bind _currentheropanel1 (Variables|Default|GetCurrentHeroPanel1)
    (:then
      (Widget|RemovefromParent _currentheropanel1)
      (Variables|Default|SetCurrentHeroPanel1 0)
      (Utilities|Platform|CollectGarbage))
    (:else)))

(event Custom|RemovePanel2_Widget
  ; create hero panel 2 widget
  (bind _currentheropanel2 (Variables|Default|GetCurrentHeroPanel2)
    (:then
      (Widget|RemovefromParent _currentheropanel2)
      (Variables|Default|SetCurrentHeroPanel2 0)
      (Utilities|Platform|CollectGarbage))
    (:else)))

(event Custom|RemovePanel3_Widget
  ; create hero panel 3 widget
  (bind _currentheropanel3 (Variables|Default|GetCurrentHeroPanel3)
    (:then
      (Widget|RemovefromParent _currentheropanel3)
      (Variables|Default|SetCurrentHeroPanel3 0)
      (Utilities|Platform|CollectGarbage))
    (:else)))

(event Custom|PlayHeroPanelAnimation
  (params (HeroPanel :type E_PistolCanary_HeroPanelSelect%20Enum :default "NewEnumerator0"))
  ; play hero panel animation
  (PistolCanary|PlayPistolCanaryHeroPanelAnimation(Message) (Variables|Default|GetCurrentPanel) HeroPanel "NewEnumerator0"))

(event Custom|PlayHeroPanel1_Animation
  (params (HeroPanel :type E_PistolCanary_HeroPanelSelect%20Enum :default "NewEnumerator0") (PanelSide :type E_PistolCanary_PanelSide%20Enum :default "NewEnumerator0"))
  ; play hero panel 1 animation
  (PistolCanary|PlayPistolCanaryHeroPanelAnimation(Message) (Variables|Default|GetCurrentHeroPanel1) HeroPanel PanelSide))

(event Custom|PlayHeroPanel2_Animation
  (params (HeroPanel :type E_PistolCanary_HeroPanelSelect%20Enum :default "NewEnumerator0") (PanelSide :type E_PistolCanary_PanelSide%20Enum :default "NewEnumerator0"))
  ; play hero panel 2 animation
  (PistolCanary|PlayPistolCanaryHeroPanelAnimation(Message) (Variables|Default|GetCurrentHeroPanel2) HeroPanel PanelSide))

(event Custom|PlayHeroPanel3_Animation
  (params (HeroPanel :type E_PistolCanary_HeroPanelSelect%20Enum :default "NewEnumerator0") (PanelSide :type E_PistolCanary_PanelSide%20Enum :default "NewEnumerator0"))
  ; play hero panel 3 animation
  (PistolCanary|PlayPistolCanaryHeroPanelAnimation(Message) (Variables|Default|GetCurrentHeroPanel3) HeroPanel PanelSide))

(event Custom|HeroPanel3_LoadTextures
  ; hero panel 3 load textures
  (PistolCanary|HeroPanelLoadTextures(Message) (Variables|Default|GetCurrentHeroPanel3)))

(event Custom|HeroPanel1_LoadTextures
  ; hero panel 1 load textures
  (PistolCanary|HeroPanelLoadTextures(Message) (Variables|Default|GetCurrentHeroPanel1)))

(event Custom|HeroPanel2_LoadTextures
  ; hero panel 1 load textures
  (PistolCanary|HeroPanelLoadTextures(Message) (Variables|Default|GetCurrentHeroPanel2)))

(event Custom|HeroPanel3_UnloadTextures
  ; hero panel 3 unload textures
  (PistolCanary|HeroPanelUnloadTextures(Message) (Variables|Default|GetCurrentHeroPanel3)))

(event Custom|HeroPanel2_UnloadTextures
  ; hero panel 2 unload textures
  (PistolCanary|HeroPanelUnloadTextures(Message) (Variables|Default|GetCurrentHeroPanel2)))

(event Custom|HeroPanel1_UnloadTextures
  ; hero panel 1 unload textures
  (PistolCanary|HeroPanelUnloadTextures(Message) (Variables|Default|GetCurrentHeroPanel1)))

(event Custom|AddBossHealthWidget
  ; create boss health bar widget
  (bind _class (Utilities|AsyncLoadClassAsset "/PistolCanary/Layout/BP/ComicPanels/PistolCanary_BossHealthBar.PistolCanary_BossHealthBar_C")
    (:Completed
      (bind _returnvalue (UserInterface|CreateWidget (Utilities|Casting|CastToUserWidgetClass _class)))
      (bind _output_get (Variables|Default|SetCurrentBossHealth _returnvalue))
      (UserInterface|Viewport|AddtoViewport _output_get 5))))

(event Custom|CreateDevWidget)

(event Custom|RemoveBossHealthWidget
  ; create boss health bar widget
  (bind _currentbosshealth (Variables|Default|GetCurrentBossHealth)
    (:then
      (Widget|RemovefromParent _currentbosshealth)
      (Variables|Default|SetCurrentBossHealth 0))
    (:else)))

(event Custom|RemoveDevWidget)

(event Custom|UpdateDevWidgetText
  (params (SegmentA :type E_PistolCanary_DevSegmentA%20Enum :default "NewEnumerator0") (SegmentB :type E_PistolCanary_DevSegmentB%20Enum :default "NewEnumerator0") (SegmentC :type E_PistolCanary_DevSegmentC%20Enum :default "NewEnumerator11") (SequencePlayer :type Level%20Sequence%20Player%20Object%20Reference)))

(event Custom|BossHealthBarVisibility
  (params (HealthBarIndex :type Integer) (Visibility :type ESlateVisibility%20Enum :default "Visible"))
  ; update boss health bar
  (PistolCanary|SetPistolCanaryBossHealthBarVisibility(Message) (Variables|Default|GetCurrentBossHealth) HealthBarIndex Visibility))

(event Custom|BossHealthBarPercent
  (params (HealthBarIndex :type Integer) (Percent :type Float%20%28double-precision%29))
  ; update boss health bar
  (PistolCanary|UpdatePistolCanaryBossHealthBar(Message) (Variables|Default|GetCurrentBossHealth) HealthBarIndex Percent))

(event Custom|Play_HP_Fill_Bar1
  (params (HealthBarIndex :type Integer))
  ; update boss health bar
  (Class|PistolCanaryBossHealthBar|AnimHPFillBar1 (Utilities|Casting|CastToPistolCanary_BossHealthBar (Variables|Default|GetCurrentBossHealth))))

(event Custom|UpdateDevWidgetTimer
  (params (InText :type Text))
  (bind _currentdevwidget (Variables|Default|GetCurrentDevWidget)
    (:then
      (bind _aspistol_canary_dev_hud (Utilities|Casting|CastToPistolCanary_DevHUD _currentdevwidget))
      (Class|PistolCanaryDevHUD|UpdateTimer _aspistol_canary_dev_hud InText))
    (:else)))

(event Custom|Play_TakeDamage_B2
  ; update boss health bar
  (Class|PistolCanaryBossHealthBar|AnimTakeDamageB2 (Utilities|Casting|CastToPistolCanary_BossHealthBar (Variables|Default|GetCurrentBossHealth))))

(event Custom|Play_TakeDamage_Bar1
  ; update boss health bar
  (Class|PistolCanaryBossHealthBar|AnimTakeDamageB1 (Utilities|Casting|CastToPistolCanary_BossHealthBar (Variables|Default|GetCurrentBossHealth))))

(event Custom|ChangeTrackedSequence
  (params (SequenceToWatch :type Level%20Sequence%20Player%20Object%20Reference))
  (Variables|Default|SetSequenceToWatch SequenceToWatch)
  (Utilities|Time|ClearandInvalidateTimerbyHandle (Variables|Default|GetTimer_UpdateTime))
  (CallFunction|StartTrackingSequenceTime))

(event Custom|Play_TakeDamage_Bar1_Long
  ; update boss health bar
  (Class|PistolCanaryBossHealthBar|AnimTakeDamageLongB1 (Utilities|Casting|CastToPistolCanary_BossHealthBar (Variables|Default|GetCurrentBossHealth))))

(event Custom|AddFullScreenPanelWidget)

(event Custom|Play_TakeDamage_B3
  ; update boss health bar
  (Class|PistolCanaryBossHealthBar|AnimTakeDamageB3 (Utilities|Casting|CastToPistolCanary_BossHealthBar (Variables|Default|GetCurrentBossHealth))))

(event Custom|PlayComicIntroMovie)

(event Custom|CustomEvent
  ; OLD play comic intro movie
  (CallFunction|RemoveFullscreenPanelWidget))

(event Custom|StartTrackingSequenceTime
  (bind _returnvalue (Utilities|Time|SetTimerbyEvent (delegate Custom|UpdateTimer) 0.1 true true))
  (Variables|Default|SetTimer_UpdateTime _returnvalue)
  (bind _currentdevwidget (Variables|Default|GetCurrentDevWidget)
    (:then
      (bind _aspistol_canary_dev_hud (Utilities|Casting|CastToPistolCanary_DevHUD _currentdevwidget))
      (Class|PistolCanaryDevHUD|UpdateTimer _aspistol_canary_dev_hud "NSLOCTEXT(\"[A7CA36EE334AE8DFDFF31F93A478A900]\", \"54E0762345764877DB105A88708D4225\", \"0\")")
      (Class|PistolCanaryDevHUD|ShowStartSignal _aspistol_canary_dev_hud)
      (Utilities|FlowControl|Delay 0.0)
      (Utilities|FlowControl|Delay 0.0)
      (Utilities|FlowControl|Delay 0.0)
      (Class|PistolCanaryDevHUD|HideStartSignal _aspistol_canary_dev_hud))
    (:else)))

(event Custom|Play_HP_Fill_Bar2
  ; update boss health bar
  (Class|PistolCanaryBossHealthBar|AnimHPFillBar2 (Utilities|Casting|CastToPistolCanary_BossHealthBar (Variables|Default|GetCurrentBossHealth))))

(event Custom|RemoveFullscreenPanelWidget)

(event Custom|UpdateTimer
  (bind _sequencetowatch (Variables|Default|GetSequenceToWatch)
    (:then
      (CallFunction|UpdateDevWidgetTimer (Utilities|Text|FormatText "NSLOCTEXT(\"[A7CA36EE334AE8DFDFF31F93A478A900]\", \"2018C2F54F92EBC30FD2E99690E25462\", \"{1} - {2}\")" (Utilities|Text|ToText(Float) (Utilities|TimeManagement|QualifiedFrameTimeToSeconds (Sequencer|Player|GetCurrentTime _sequencetowatch)) "HalfToEven" false true 1 324 2 2) (Utilities|Text|ToText(Float) (- (Utilities|TimeManagement|QualifiedFrameTimeToSeconds (Sequencer|Player|GetEndTime _sequencetowatch :!class /Script/MovieScene.MovieSceneSequencePlayer)) (Utilities|TimeManagement|QualifiedFrameTimeToSeconds (Sequencer|Player|GetCurrentTime _sequencetowatch))) "HalfToEven" false true 1 324 2 2))))
    (:else)))

(event Custom|Play_HP_Fill_Bar3
  ; update boss health bar
  (Class|PistolCanaryBossHealthBar|AnimHPFillBar3 (Utilities|Casting|CastToPistolCanary_BossHealthBar (Variables|Default|GetCurrentBossHealth))))

(event Custom|OpenComicIntroMedia)

(event Custom|PlayIntroMovieWipe
  ; play intro movie wipe
  (PistolCanary|PlayPistolCanaryIntroWipeAnimation(Message) (Variables|Default|GetCurrentFullScreenPanel)))

(event Custom|LoadIntro_v2_Panel1)

(event Custom|UnloadIntro_v2_Panel1)

(fn ConstructionScript)
